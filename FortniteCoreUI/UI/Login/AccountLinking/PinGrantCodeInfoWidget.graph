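# Login account-linking panel that tells the player to visit the Epic activation URL and enter a PIN code.
# It lays out the instructions, a backplated area for the PIN or a server error, a throbber, and a footer with an expiry timer and Refresh Code button.
# The error, throbber and timer start collapsed.
# The Blueprint only supplies layout and styling for widgets inherited from FortPinGrantCodeInfoWidget, with no logic in its event graph.
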WidgetBlueprint /FortniteCoreUI/UI/Login/AccountLinking/PinGrantCodeInfoWidget
graphs [EventGraph {nodes [
   Event PreConstruct()
   Event Construct()
   Event Tick()
  ]}]
variables []
widgets {
 info {inheritedWidgetCount=9, parentClass=/Script/FortniteUI.FortPinGrantCodeInfoWidget, rootWidgetClass=&VerticalBox, widgetCount=17}
 widgets [
  VerticalBoxHaveEpicAccount {widgetClassPath=&VerticalBox, widget=VerticalBoxHaveEpicAccount}
  Container_ActivationURL {
   widgetClassPath=&VerticalBox
   bInherited=true
   parent=VerticalBoxHaveEpicAccount
   widget=Container_ActivationURL
   slot {class=VerticalBoxSlot}
  }
  Text_PinPromptTypeTitle {
   widgetClassPath=&CommonTextBlock
   bInherited=true
   parent=Container_ActivationURL
   widget=Text_PinPromptTypeTitle
   properties {
    applyLineHeightToBottomLine=false
    justification=Center
    text=link an account
    textTransformPolicy=ToUpper
    font {
     bForceMonospaced=false
     bMaterialIsStencil=false
     fontMaterial=null
     fontObject=&Burbank
     letterSpacing=0
     monospacedWidth=1
     size=32
     skewAmount=0
     typefaceFontName=Burbank-Black
     outlineSettings {
      bApplyOutlineToDropShadows=false
      bMiteredCorners=false
      bSeparateFillAlpha=false
      outlineMaterial=null
      outlineSize=0
      outlineColor=<rgba 0 0 0 1>
     }
    }
    shadowOffset=<0 0>
   }
   slot {class=VerticalBoxSlot, properties {horizontalAlignment=HAlign_Center, verticalAlignment=VAlign_Top, padding=<ltrb 0 0 0 16>}}
  }
  HorizontalBox_317 {
   widgetClassPath=&HorizontalBox
   parent=Container_ActivationURL
   widget=HorizontalBox_317
   slot {class=VerticalBoxSlot, properties {padding=<ltrb 0 0 0 8>}}
  }
  CommonTextBlock_68 {
   widgetClassPath=&CommonTextBlock
   parent=HorizontalBox_317
   widget=CommonTextBlock_68
   properties {
    text="1. Visit "
    font {
     bForceMonospaced=false
     bMaterialIsStencil=false
     fontMaterial=null
     fontObject=&Burbank
     letterSpacing=0
     monospacedWidth=1
     size=32
     skewAmount=0
     typefaceFontName=Burbank-Black
     outlineSettings {
      bApplyOutlineToDropShadows=true
      bMiteredCorners=false
      bSeparateFillAlpha=true
      outlineMaterial=&M-FontShadow
      outlineSize=3
      outlineColor=<rgba 0 0 0 0.5>
     }
    }
    shadowOffset=<0 0>
   }
   slot {class=HorizontalBoxSlot, properties {horizontalAlignment=HAlign_Left, verticalAlignment=VAlign_Center}}
  }
  Text_ActivationURL {
   widgetClassPath=&CommonTextBlock
   bInherited=true
   parent=HorizontalBox_317
   widget=Text_ActivationURL
   properties {
    text=https://www.epicgames.com/activate
    font {
     bForceMonospaced=false
     bMaterialIsStencil=false
     fontMaterial=null
     fontObject=&Burbank
     letterSpacing=0
     monospacedWidth=1
     size=32
     skewAmount=0
     typefaceFontName=Burbank-Black
     outlineSettings {
      bApplyOutlineToDropShadows=true
      bMiteredCorners=false
      bSeparateFillAlpha=true
      outlineMaterial=&M-FontShadow
      outlineSize=3
      outlineColor=<rgba 0 0 0 0.5>
     }
    }
    shadowOffset=<0 0>
   }
   slot {class=HorizontalBoxSlot, properties {horizontalAlignment=HAlign_Left, verticalAlignment=VAlign_Center}}
  }
  CommonTextBlock_69 {
   widgetClassPath=&CommonTextBlock
   parent=Container_ActivationURL
   widget=CommonTextBlock_69
   properties {
    style=/FortUILibrary/Styles/Typography/Foundation/TextStyle-Header-M-S.TextStyle-Header-M-S_C
    text=2. Enter the following code:
    font {
     bForceMonospaced=false
     bMaterialIsStencil=false
     fontMaterial=null
     fontObject=&Burbank
     letterSpacing=0
     monospacedWidth=1
     size=32
     skewAmount=0
     typefaceFontName=Burbank-Black
     outlineSettings {
      bApplyOutlineToDropShadows=true
      bMiteredCorners=false
      bSeparateFillAlpha=true
      outlineMaterial=&M-FontShadow
      outlineSize=3
      outlineColor=<rgba 0 0 0 0.5>
     }
    }
    shadowOffset=<0 0>
   }
   slot {class=VerticalBoxSlot, properties {horizontalAlignment=HAlign_Left, verticalAlignment=VAlign_Top}}
  }
  B_Pin {
   widgetClassPath=&Overlay
   parent=VerticalBoxHaveEpicAccount
   widget=B_Pin
   slot {class=VerticalBoxSlot, properties {verticalAlignment=VAlign_Center, padding=<ltrb 0 32 0 32>, size {sizeRule=Fill, value=1}}}
  }
  WBP_UIKit_Backplate {
   widgetClassPath=&WBP_UIKit_Backplate
   bIsVariable=true
   parent=B_Pin
   widget=WBP_UIKit_Backplate
   properties {brightness=Dark}
   slot {class=OverlaySlot, properties {horizontalAlignment=HAlign_Fill, verticalAlignment=VAlign_Fill}}
  }
  PinAndThrobberContainer {
   widgetClassPath=&HorizontalBox
   parent=B_Pin
   widget=PinAndThrobberContainer
   slot {class=OverlaySlot, properties {horizontalAlignment=HAlign_Fill, verticalAlignment=VAlign_Fill, padding=<ltrb 24 24 24 24>}}
  }
  Text_ErrorMsg {
   widgetClassPath=&CommonTextBlock
   bInherited=true
   parent=PinAndThrobberContainer
   widget=Text_ErrorMsg
   properties {
    justification=Center
    style=/FortUILibrary/Styles/Typography/Foundation/TextStyle-Header-M.TextStyle-Header-M_C
    text="Sorry, We are having a problems reaching Epic Games servers."
    visibility=Collapsed
    font {
     bForceMonospaced=false
     bMaterialIsStencil=false
     fontMaterial=null
     fontObject=&Burbank
     letterSpacing=0
     monospacedWidth=1
     size=32
     skewAmount=0
     typefaceFontName=Burbank-Black
     outlineSettings {
      bApplyOutlineToDropShadows=false
      bMiteredCorners=false
      bSeparateFillAlpha=false
      outlineMaterial=null
      outlineSize=0
      outlineColor=<rgba 0 0 0 1>
     }
    }
    shadowOffset=<0 0>
   }
   slot {class=HorizontalBoxSlot, properties {horizontalAlignment=HAlign_Center, verticalAlignment=VAlign_Center, size {sizeRule=Fill, value=1}}}
  }
  Text_Pincode {
   widgetClassPath=&CommonTextBlock
   bInherited=true
   parent=PinAndThrobberContainer
   widget=Text_Pincode
   properties {
    justification=Center
    text=0|NV4lID1O
    font {
     bForceMonospaced=false
     bMaterialIsStencil=false
     fontMaterial=null
     fontObject=&HeadingNow
     letterSpacing=0
     monospacedWidth=1
     size=51.75
     skewAmount=0
     typefaceFontName=76 Bold
     outlineSettings {
      bApplyOutlineToDropShadows=false
      bMiteredCorners=false
      bSeparateFillAlpha=false
      outlineMaterial=null
      outlineSize=0
      outlineColor=<rgba 0 0 0 1>
     }
    }
   }
   slot {class=HorizontalBoxSlot, properties {horizontalAlignment=HAlign_Center, verticalAlignment=VAlign_Center, size {sizeRule=Fill, value=1}}}
  }
  Throbber_GettingPinCode {
   widgetClassPath=/Script/UMG.CircularThrobber
   bInherited=true
   bIsVariable=true
   parent=VerticalBoxHaveEpicAccount
   widget=Throbber_GettingPinCode
   properties {radius=32, visibility=Collapsed}
   slot {class=VerticalBoxSlot, properties {horizontalAlignment=HAlign_Center, verticalAlignment=VAlign_Top, padding=<ltrb 0 0 0 24>}}
  }
  Text_Info {
   widgetClassPath=&CommonTextBlock
   bInherited=true
   parent=VerticalBoxHaveEpicAccount
   widget=Text_Info
   properties {
    applyLineHeightToBottomLine=false
    justification=Center
    text="Stay on this screen while you sign in, it will refresh once you're done"
    wrapTextAt=944
    font {
     bForceMonospaced=false
     bMaterialIsStencil=false
     fontMaterial=null
     fontObject=&Burbank
     letterSpacing=0
     monospacedWidth=1
     size=32
     skewAmount=0
     typefaceFontName=Burbank-Black
     outlineSettings {
      bApplyOutlineToDropShadows=true
      bMiteredCorners=false
      bSeparateFillAlpha=true
      outlineMaterial=&M-FontShadow
      outlineSize=3
      outlineColor=<rgba 0 0 0 0.5>
     }
    }
    shadowOffset=<0 0>
   }
   slot {class=VerticalBoxSlot, properties {horizontalAlignment=HAlign_Center, verticalAlignment=VAlign_Top}}
  }
  CodeBtnContainer {
   widgetClassPath=&HorizontalBox
   parent=VerticalBoxHaveEpicAccount
   widget=CodeBtnContainer
   slot {class=VerticalBoxSlot, properties {horizontalAlignment=HAlign_Center, verticalAlignment=VAlign_Top, padding=<ltrb 0 32 0 0>}}
  }
  Text_CodeExpiringTime {
   widgetClassPath=&CommonTextBlock
   bInherited=true
   parent=CodeBtnContainer
   widget=Text_CodeExpiringTime
   properties {
    text=Expiring in 2:34 sec.
    textTransformPolicy=ToUpper
    visibility=Collapsed
    colorAndOpacity {colorUseRule=UseColor_Specified, specifiedColor=<rgba 0.9301 1 0.0103 1>}
    font {
     bForceMonospaced=false
     bMaterialIsStencil=false
     fontMaterial=null
     fontObject=&Burbank
     letterSpacing=0
     monospacedWidth=1
     size=32
     skewAmount=0
     typefaceFontName=Burbank-Black
     outlineSettings {
      bApplyOutlineToDropShadows=false
      bMiteredCorners=false
      bSeparateFillAlpha=false
      outlineMaterial=null
      outlineSize=0
      outlineColor=<rgba 0 0 0 1>
     }
    }
    shadowOffset=<0 0>
   }
   slot {class=HorizontalBoxSlot, properties {horizontalAlignment=HAlign_Center, verticalAlignment=VAlign_Center, size {sizeRule=Fill, value=1}}}
  }
  Button_GenerateCode {
   widgetClassPath=&IconTextButton
   bInherited=true
   bIsVariable=true
   parent=CodeBtnContainer
   widget=Button_GenerateCode
   properties {
    buttonText=Refresh Code
    onButtonBaseClicked=(&self:WidgetTree.Button_GenerateCode.HandleOnButtonClicked)
    onButtonBaseDoubleClicked=(&self:WidgetTree.Button_GenerateCode.HandleOnButtonDoubleClicked)
    onButtonBaseHovered=(&self:WidgetTree.Button_GenerateCode.HandleOnButtonHovered)
    onButtonBaseUnhovered=(&self:WidgetTree.Button_GenerateCode.HandleOnButtonUnhovered)
    onSelectedChangedBase=(&self:WidgetTree.Button_GenerateCode.HandleOnSelectedChanged)
    style=/FortUILibrary/Foundation/Buttons/ButtonStyles/ButtonStyle_Feature_L_Yellow.ButtonStyle_Feature_L_Yellow_C
   }
   slot {class=HorizontalBoxSlot, properties {horizontalAlignment=HAlign_Center, verticalAlignment=VAlign_Center}}
  }
 ]
}
